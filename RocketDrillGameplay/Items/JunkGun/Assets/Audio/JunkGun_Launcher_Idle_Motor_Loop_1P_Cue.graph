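# First-person SoundCue for the JunkGun launcher's idling motor, with distance attenuation and a concurrency limit.
# After a short delay, a quality switch picks either a full mix (first two levels) or the putter loop alone (last two).
# The delayed putter loop runs through volume and pitch envelopes.
# The full mix adds endlessly repeating, randomly delayed backfire layers: one of ten pops and a single poof, both with random volume and pitch.
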SoundCue /RocketDrillGameplay/Items/JunkGun/Assets/Audio/JunkGun_Launcher_Idle_Motor_Loop_1P_Cue
root=Delay_1
nodes [
 {id=Delay_0, children [Modulator_0], properties {delayMax=3, delayMin=0.5}}
 {id=Delay_1, children [QualityLevel_0], properties {delayMax=0.2, delayMin=0.2}}
 {id=Delay_3, children [WavePlayer_1], properties {delayMax=1, delayMin=1}}
 {id=Delay_4, children [WavePlayer_21], properties {delayMax=1, delayMin=1}}
 {id=Delay_6, children [Modulator_1], properties {delayMax=2}}
 {
  id=Enveloper_0
  children [Delay_4]
  properties {
   pitchCurve {editorCurveData {keys [{time=0}, <at 1.5547 1.0022>, <at 9.348 0.9303>]}}
   volumeCurve {editorCurveData {keys [
      {arriveTangent=3222.0066, interpMode=RCIM_Cubic, leaveTangent=4.8, tangentMode=RCTM_User, time=0, value=0}
      {arriveTangent=0.9018, interpMode=RCIM_Cubic, leaveTangent=0.9018, tangentMode=RCTM_User, time=1.4865, value=4.1097}
      <at 4.9247 4.8833>
      {arriveTangent=-0.2813, leaveTangent=-0.2813, time=14.8838, value=2.283}
     ]}}
  }
 }
 {
  id=Enveloper_1
  children [Delay_3]
  properties {
   pitchCurve {editorCurveData {keys [{time=0}, <at 1.5547 1.0022>, <at 9.348 0.9303>]}}
   volumeCurve {editorCurveData {keys [
      {arriveTangent=3222.0066, interpMode=RCIM_Cubic, leaveTangent=4.8, tangentMode=RCTM_User, time=0, value=0}
      {arriveTangent=0.9018, interpMode=RCIM_Cubic, leaveTangent=0.9018, tangentMode=RCTM_User, time=1.4865, value=4.1097}
      <at 4.9247 4.8833>
      {arriveTangent=-0.2813, leaveTangent=-0.2813, time=14.8838, value=2.283}
     ]}}
  }
 }
 {id=Looping_0, children [Delay_0]}
 {id=Looping_1, children [Delay_6]}
 {id=Mixer_0, children [Enveloper_0], properties {inputVolume [0.22]}}
 {id=Mixer_1, children [Enveloper_1,Looping_0,Looping_1], properties {inputVolume [0.18,0.2,0.4]}}
 {id=Modulator_0, children [Random_1], properties {pitchMax=1.5, pitchMin=0.8, volumeMax=1, volumeMin=0}}
 {id=Modulator_1, children [WavePlayer_20], properties {pitchMax=1.5, volumeMax=1, volumeMin=0}}
 {id=QualityLevel_0, children [Mixer_1,Mixer_1,Mixer_0,Mixer_0]}
 {
  id=Random_1
  children [WavePlayer_2, WavePlayer_3, WavePlayer_4, WavePlayer_5, WavePlayer_6, WavePlayer_7, WavePlayer_8, WavePlayer_9, WavePlayer_10, WavePlayer_11]
  properties {weights [1,1,1,1,1,1,1,1,1,1]}
 }
 {
  id=WavePlayer_1
  properties {bLooping=true, soundWaveAssetPtr=&/JunkGun_Launcher_Idle_Motor_Putter_Loop_F}
 }
 {id=WavePlayer_10, properties {soundWaveAssetPtr=&/JunkGun_Launcher_Motor_Backfire_Pop_08}}
 {id=WavePlayer_11, properties {soundWaveAssetPtr=&/JunkGun_Launcher_Motor_Backfire_Pop_09}}
 {id=WavePlayer_2, properties {soundWaveAssetPtr=&/JunkGun_Launcher_Motor_Backfire_Pop_10}}
 {
  id=WavePlayer_20
  properties {soundWaveAssetPtr=&/JunkGun_Launcher_Motor_Loop_Backfire_Poof_01}
 }
 {
  id=WavePlayer_21
  properties {bLooping=true, soundWaveAssetPtr=&JunkGun_Launcher_Idle_Motor_Putter_Loop_F}
 }
 {id=WavePlayer_3, properties {soundWaveAssetPtr=&/JunkGun_Launcher_Motor_Backfire_Pop_01}}
 {id=WavePlayer_4, properties {soundWaveAssetPtr=&/JunkGun_Launcher_Motor_Backfire_Pop_02}}
 {id=WavePlayer_5, properties {soundWaveAssetPtr=&/JunkGun_Launcher_Motor_Backfire_Pop_03}}
 {id=WavePlayer_6, properties {soundWaveAssetPtr=&/JunkGun_Launcher_Motor_Backfire_Pop_04}}
 {id=WavePlayer_7, properties {soundWaveAssetPtr=&/JunkGun_Launcher_Motor_Backfire_Pop_05}}
 {id=WavePlayer_8, properties {soundWaveAssetPtr=&/JunkGun_Launcher_Motor_Backfire_Pop_06}}
 {id=WavePlayer_9, properties {soundWaveAssetPtr=&/JunkGun_Launcher_Motor_Backfire_Pop_07}}
]
properties {
 attenuationSettings=&/Fort_Weapon_JunkGun_Motor_Athena_Att
 bOverrideConcurrency=true
 maxDistance=8100
 soundClassObject=&Fort_WeaponFoley
 volumeMultiplier=0.501
 concurrencyOverrides {maxCount=3}
}
